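# Physics asset for the CashewRoll backpack cosmetic, defining simulation bodies for its dangling left and right belts.
# Each side has three small capsules on the master, 01 and 02 bones, beside one large capsule for the pack body.
# Angular-limited constraints chain each side's segments with very little twist, and the left side swings more freely than the right.
# No constraint links either master or the pack to another body.
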PhysicsAsset /BRCosmetics/Accessories/FORT_Backpacks/Backpack_M_MED_CashewRoll/Meshes/M_MED_CashewRoll_Pack_BeltsPhysics
bodies {
 dyn_belt_l_01 [{shapeType=Capsule, length=1.3215, radius=0.6534, shapeName=dyn_belt_l_01_capsule, center=<0.9161 0 0.3202>, rotation=<rot -90 -180 180>}]
 dyn_belt_l_02 [{
   shapeType=Capsule
   length=1.3215
   radius=0.6534
   shapeName=dyn_belt_l_02_capsule
   center=<0.9161 0 0.3202>
   rotation=<rot -90 -179.6835 179.6835>
  }]
 dyn_belt_l_master [{
   shapeType=Capsule
   length=1.3215
   radius=0.6534
   shapeName=dyn_belt_l_master_capsule
   center=<0.9161 0 0.3202>
   rotation=<rot -90 -173.2786 173.2786>
  }]
 dyn_belt_r_01 [{
   shapeType=Capsule
   length=1.3215
   radius=0.6534
   shapeName=dyn_belt_r_01_capsule
   center=<0.9161 0 0.3202>
   rotation=<rot -90 -179.9209 179.9209>
  }]
 dyn_belt_r_02 [{
   shapeType=Capsule
   length=1.3215
   radius=0.6534
   shapeName=dyn_belt_r_02_capsule
   center=<0.9161 0 0.3202>
   rotation=<rot -90 -7.5524 7.5524>
  }]
 dyn_belt_r_master [{
   shapeType=Capsule
   length=1.3215
   radius=0.6534
   shapeName=dyn_belt_r_master_capsule
   center=<0.9161 0 0.3202>
   rotation=<rot -90 -173.2786 173.2786>
  }]
 pack [{shapeType=Capsule, length=15, radius=15, shapeName=pack_capsule, center=<0 -0.8944 -16.9377>, rotation=<rot 90 -160.5288 -160.5289>}]
}
constraints [
 {bone1Name=dyn_belt_r_02, bone2Name=dyn_belt_r_01, swing1LimitDegrees=3, swing2LimitDegrees=6, twistLimitDegrees=1}
 {bone1Name=dyn_belt_r_01, bone2Name=dyn_belt_r_master, swing1LimitDegrees=3, swing2LimitDegrees=6, twistLimitDegrees=1}
 {bone1Name=dyn_belt_l_02, bone2Name=dyn_belt_l_01, swing1LimitDegrees=7, swing2LimitDegrees=15, twistLimitDegrees=1}
 {bone1Name=dyn_belt_l_01, bone2Name=dyn_belt_l_master, swing1LimitDegrees=7, swing2LimitDegrees=15, twistLimitDegrees=1}
]
