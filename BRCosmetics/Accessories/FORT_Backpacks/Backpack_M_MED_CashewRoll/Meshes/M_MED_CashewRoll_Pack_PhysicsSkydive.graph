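# Physics asset for the CashewRoll backpack cosmetic, named as the skydive variant, covering its dangling scarf, belts and pack body.
# Capsules follow a scarf chain from the master bone, forking after the second segment into two three-segment tails, linked by constraints with limited swing and tight twist.
# The pack uses two capsules and a box.
# The left and right belt bones each have a capsule but no constraints.
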PhysicsAsset /BRCosmetics/Accessories/FORT_Backpacks/Backpack_M_MED_CashewRoll/Meshes/M_MED_CashewRoll_Pack_PhysicsSkydive
bodies {
 dyn_belt_l_01 [{shapeType=Capsule, length=4.5326, radius=1.7049, shapeName=dyn_belt_l_01_capsule, center=<2.0117 0 -0.7401>, rotation=<rot -90 0 0>}]
 dyn_belt_r_01 [{shapeType=Capsule, length=4.5326, radius=1.7049, shapeName=dyn_belt_r_01_capsule, center=<2.0117 0 -0.7401>, rotation=<rot -90 0 0>}]
 dyn_scarf_01 [{shapeType=Capsule, length=1.0727, radius=1.9383, shapeName=dyn_scarf_01_capsule, center=<1.7062 0 0>, rotation=<rot -90 0.0005 -0.0005>}]
 dyn_scarf_02 [{shapeType=Capsule, length=0.3301, radius=2.6809, shapeName=dyn_scarf_02_capsule, center=<1.7062 0 0>, rotation=<rot -90 179.9802 -179.9803>}]
 dyn_scarf_03_a [{
   shapeType=Capsule
   length=2.2756
   radius=1.4491
   shapeName=dyn_scarf_03_a_capsule
   center=<1.6685 0 -0.4585>
   rotation=<rot -78.9627 114.9021 -115.3133>
  }]
 dyn_scarf_03_b [
  {
   shapeType=Capsule
   length=1.5619
   radius=1.9212
   shapeName=dyn_scarf_03_b_capsule
   center=<1.7062 0 0.8012>
   rotation=<rot -90 -180 -179.9999>
  }
  {
   shapeType=Capsule
   length=1.5619
   radius=1.8737
   shapeName=dyn_scarf_03_b_capsule
   center=<1.8953 -2.4226 -0.2365>
   rotation=<rot -90 -180 -179.9999>
  }
  {
   shapeType=Capsule
   length=1.5619
   radius=1.858
   shapeName=dyn_scarf_03_b_capsule
   center=<1.8503 -1.2748 -0.0761>
   rotation=<rot -90 -180 -179.9999>
  }
 ]
 dyn_scarf_04_a [{
   shapeType=Capsule
   length=1.9672
   radius=1.6906
   shapeName=dyn_scarf_04_a_capsule
   center=<1.9082 0 -0.3516>
   rotation=<rot -90 179.9999 179.9999>
  }]
 dyn_scarf_04_b [{
   shapeType=Capsule
   length=0.8355
   radius=2.1755
   shapeName=dyn_scarf_04_b_capsule
   center=<1.7062 0 0.0054>
   rotation=<rot -90 -6.7214 6.7214>
  }]
 dyn_scarf_05_a [{
   shapeType=Capsule
   length=1.9657
   radius=1.0454
   shapeName=dyn_scarf_05_a_capsule
   center=<1.358 0.0086 0.2373>
   rotation=<rot -80.1448 -12.0863 12.2617>
  }]
 dyn_scarf_05_b [{shapeType=Capsule, length=1.5619, radius=1.4491, shapeName=dyn_scarf_05_b_capsule, center=<1.7062 0 0.2935>, rotation=<rot -90 -180 180>}]
 dyn_scarf_master [{shapeType=Capsule, length=0, radius=0.7945, shapeName=dyn_scarf_master_capsule, center=<-0.6822 0 0>, rotation=<rot -40.0402 180 -180>}]
 pack [
  {shapeType=Capsule, length=22.1738, radius=14.7231, shapeName=pack_capsule, center=<0 2.9126 -6.1948>, rotation=<rot 90 -179.9802 -179.9803>}
  {shapeType=Capsule, length=15, radius=15, shapeName=pack_capsule, center=<0 7.8206 20.5024>, rotation=<rot 0 0 -35>}
  {shapeType=Box, extentX=22.559, extentY=7.5, extentZ=18.0493, shapeName=pack_box, center=<0 -9.1929 -13.1177>, rotation=<rot 0 0 10.418>}
 ]
}
constraints [
 {bone1Name=dyn_scarf_01, bone2Name=dyn_scarf_master, swing1LimitDegrees=25, swing2LimitDegrees=40, twistLimitDegrees=5}
 {bone1Name=dyn_scarf_02, bone2Name=dyn_scarf_01, swing1LimitDegrees=30, swing2LimitDegrees=40, twistLimitDegrees=5}
 {bone1Name=dyn_scarf_03_a, bone2Name=dyn_scarf_02, swing1LimitDegrees=35, swing2LimitDegrees=35, twistLimitDegrees=5}
 {bone1Name=dyn_scarf_04_a, bone2Name=dyn_scarf_03_a, swing1LimitDegrees=35, swing2LimitDegrees=35, twistLimitDegrees=5}
 {bone1Name=dyn_scarf_05_a, bone2Name=dyn_scarf_04_a, swing1LimitDegrees=35, swing2LimitDegrees=35, twistLimitDegrees=5}
 {bone1Name=dyn_scarf_03_b, bone2Name=dyn_scarf_02, swing1LimitDegrees=25, swing2LimitDegrees=35, twistLimitDegrees=5}
 {bone1Name=dyn_scarf_04_b, bone2Name=dyn_scarf_03_b, swing1LimitDegrees=15, swing2LimitDegrees=35, twistLimitDegrees=5}
 {bone1Name=dyn_scarf_05_b, bone2Name=dyn_scarf_04_b, swing1LimitDegrees=15, swing2LimitDegrees=35, twistLimitDegrees=5}
]
