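# Physics asset named as the cloth colliders for the CashewRoll medium male backpack cosmetic.
# It defines a single body on the pack bone and no constraints.
# That body lists no collision shapes, so the file shows no actual collider geometry for cloth to collide against.
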PhysicsAsset /BRCosmetics/Accessories/FORT_Backpacks/Backpack_M_MED_CashewRoll/Meshes/M_MED_CashewRoll_Pack_ClothColliders
bodies {pack []}
constraints []
